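# Physics asset for the Jet Ski backpack cosmetic, giving its dangling parts simulated secondary motion.
# The pack bone carries a capsule and box body, while handle, radio, antenna, sickle and zipper each use a single capsule.
# Angular-limited constraints attach handle, radio, sickle and zipper to the pack, with the antenna chained off the radio.
# Limits vary per part, from a loose handle to a zipper that swings on one axis only.
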PhysicsAsset /BRCosmetics/Accessories/FORT_Backpacks/F_MED_Jet_Ski_Backpack/Meshes/F_MED_Jet_Ski_Pack_Physics
bodies {
 dyn_antenna [{shapeType=Capsule, length=15.4842, radius=0.7382, center=<7.7008 0.0905 -0.1321>, rotation=<rot 90 0 0>}]
 dyn_handle [{shapeType=Capsule, length=3.6046, radius=1.8522, center=<0.029 0.4728 -0.0101>, rotation=<rot 0 0 -90>}]
 dyn_radio [{shapeType=Capsule, length=10.7354, radius=3.7071, center=<5.2636 -1.7177 -0.3201>, rotation=<rot 90 0 0>}]
 dyn_sickle [{shapeType=Capsule, length=6.3371, radius=1.9473, center=<-4.224 1.217 -1.404>, rotation=<rot 90 0 0>}]
 dyn_zipper [{shapeType=Capsule, length=1.7693, radius=0.6781, center=<0.0112 -0.0342 -1.1679>}]
 pack [
  {shapeType=Capsule, length=14.789, radius=6.3688, center=<0.0282 -3.9922 -2.8079>}
  {shapeType=Box, extentX=7.5, extentY=6.6251, extentZ=7.5, center=<5.3919 -6.3355 3.133>}
 ]
}
constraints [
 {bone1Name=dyn_handle, bone2Name=pack, swing1LimitDegrees=45, swing2LimitDegrees=45, twistLimitDegrees=45}
 {bone1Name=dyn_radio, bone2Name=pack, swing1LimitDegrees=10, swing2LimitDegrees=12, twistLimitDegrees=15}
 {bone1Name=dyn_antenna, bone2Name=dyn_radio, swing1LimitDegrees=15, swing2LimitDegrees=15, twistLimitDegrees=3}
 {bone1Name=dyn_sickle, bone2Name=pack, swing1LimitDegrees=6, swing2LimitDegrees=20, twistLimitDegrees=6}
 {bone1Name=dyn_zipper, bone2Name=pack, swing1LimitDegrees=0, swing2LimitDegrees=65, twistLimitDegrees=0}
]
